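# Sound cue for a leaking propane tank gadget, playing one of three equally weighted leak variations at random.
# The random pick is routed to the gadgets sound class, then given slight pitch variation by a modulator.
# An attenuation node based on the RPG explosion settings overrides them to add distance low-pass filtering.
# The cue lowers its overall volume and overrides concurrency to cap simultaneous instances with volume scaling.
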SoundCue /Game/Sounds/Fort_Gadgets/PropaneTank/PropaneTank_Leak_Cue
root=Attenuation_0
nodes [
 {
  id=Attenuation_0
  children [Modulator_1]
  properties {
   attenuationSettings=/Game/Sounds/Attenuation/Fort_Explosion_RPG_Att
   bOverrideAttenuation=true
   attenuationOverrides {bAttenuateWithLPF=true, lPFFrequencyAtMax=665.3513, lPFRadiusMax=2786.9099, lPFRadiusMin=2024.5316, attenuationShapeExtents {x=1407.6522}}
  }
 }
 {id=Modulator_1, children [SoundClass_1], properties {pitchMax=1.1, pitchMin=0.9}}
 {id=Random_0, children [WavePlayer_0,WavePlayer_2,WavePlayer_3], properties {weights [1,1,1]}}
 {id=SoundClass_1, children [Random_0], properties {soundClassOverride=&Fort_Gadgets}}
 {id=WavePlayer_0, properties {soundWaveAssetPtr=&/PropaneTank_Leak_01}}
 {id=WavePlayer_2, properties {soundWaveAssetPtr=&/PropaneTank_Leak_02}}
 {id=WavePlayer_3, properties {soundWaveAssetPtr=&/PropaneTank_Leak_03}}
]
properties {bOverrideConcurrency=true, maxDistance=5007.6523, volumeMultiplier=0.4, concurrencyOverrides {maxCount=10, volumeScale=0.819}}
